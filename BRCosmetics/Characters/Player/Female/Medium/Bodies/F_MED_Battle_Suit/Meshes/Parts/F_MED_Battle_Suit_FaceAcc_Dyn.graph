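# Physics asset for the face accessory mesh of the female medium Battle Suit player cosmetic, defining physics for its hair bun and left and right tassels.
# Two-capsule bun and three-box tassel chains on each side hang from a head body built from three capsules.
# Constraints link each chain segment to its parent with tight twist limits and moderate swing cones, slightly wider where tassels meet the head.
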PhysicsAsset /BRCosmetics/Characters/Player/Female/Medium/Bodies/F_MED_Battle_Suit/Meshes/Parts/F_MED_Battle_Suit_FaceAcc_Dyn
bodies {
 dyn_hair_bun_01 [{shapeType=Capsule, length=2, radius=1, center=<0.1 0 0>, rotation=<rot 90 0 0>}]
 dyn_hair_bun_02 [{shapeType=Capsule, length=2, radius=1, center=<0.1 0 0>, rotation=<rot 90 0 0>}]
 dyn_hair_tassels_l_01 [{shapeType=Box, extentX=4.875, extentY=1.75, extentZ=0.75, center=<2.3011 0 0>}]
 dyn_hair_tassels_l_02 [{shapeType=Box, extentX=3.6875, extentY=1.75, extentZ=0.75, center=<1.7442 0 0>}]
 dyn_hair_tassels_l_03 [{shapeType=Box, extentX=3, extentY=1.75, extentZ=0.75, center=<1.3988 0 0>}]
 dyn_hair_tassels_r_01 [{shapeType=Box, extentX=4.875, extentY=1.75, extentZ=0.75, center=<-2.3011 0 0>}]
 dyn_hair_tassels_r_02 [{shapeType=Box, extentX=3.6875, extentY=1.75, extentZ=0.75, center=<-1.7442 0 0>}]
 dyn_hair_tassels_r_03 [{shapeType=Box, extentX=3, extentY=1.75, extentZ=0.75, center=<-1.3988 0 0>}]
 head [
  {shapeType=Capsule, length=7.2812, radius=5.875, center=<2.7913 3.4563 0>, rotation=<rot 90 0 0>}
  {shapeType=Capsule, length=6.4062, radius=1.375, center=<3.7607 -0.4062 -6.3135>, rotation=<rot -79.994 180 -180>}
  {shapeType=Capsule, length=6.4062, radius=1.375, center=<3.7544 -0.4063 6.3492>, rotation=<rot 79.9937 180 180>}
 ]
}
constraints [
 {bone1Name=dyn_hair_tassels_l_02, bone2Name=dyn_hair_tassels_l_01, swing1LimitDegrees=45, swing2LimitDegrees=45, twistLimitDegrees=5}
 {bone1Name=dyn_hair_tassels_l_03, bone2Name=dyn_hair_tassels_l_02, swing1LimitDegrees=45, swing2LimitDegrees=45, twistLimitDegrees=5}
 {bone1Name=dyn_hair_bun_02, bone2Name=dyn_hair_bun_01, swing1LimitDegrees=45, swing2LimitDegrees=45, twistLimitDegrees=5}
 {bone1Name=dyn_hair_tassels_r_02, bone2Name=dyn_hair_tassels_r_01, swing1LimitDegrees=45, swing2LimitDegrees=45, twistLimitDegrees=5}
 {bone1Name=dyn_hair_tassels_r_03, bone2Name=dyn_hair_tassels_r_02, swing1LimitDegrees=45, swing2LimitDegrees=45, twistLimitDegrees=5}
 {bone1Name=dyn_hair_tassels_l_01, bone2Name=head, swing1LimitDegrees=65, swing2LimitDegrees=65, twistLimitDegrees=5}
 {bone1Name=dyn_hair_bun_01, bone2Name=head, swing1LimitDegrees=45, swing2LimitDegrees=45, twistLimitDegrees=5}
 {bone1Name=dyn_hair_tassels_r_01, bone2Name=head, swing1LimitDegrees=65, swing2LimitDegrees=65, twistLimitDegrees=5}
]
